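# Physics asset for the VeiledSeer male medium player body, set up around the swinging panels of its jacket.
# Capsules cover the lower spine, pelvis, thighs and calves, plus front, back and side jacket panels, each hung from a small zero-length base capsule.
# Panels have tight swing and twist limits to their bases.
# Back and side bases join the pelvis with free rotation, while front bases have no pelvis constraint.
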PhysicsAsset /BRCosmetics/Characters/Player/Male/Medium/Bodies/M_MED_VeiledSeer/Meshes/M_MED_VeiledSeer_Physics
bodies {
 calf_l [{
   shapeType=Capsule
   length=30.4793
   radius=5.1608
   shapeName=calf_l_capsule
   center=<-20.1471 -1.1095 0.5066>
   rotation=<rot -6.2521 85.1286 -87.8972>
  }]
 calf_r [{
   shapeType=Capsule
   length=30.7184
   radius=5.143
   shapeName=calf_r_capsule
   center=<20.3971 1.2361 0.0111>
   rotation=<rot -5.775 88.0231 -88.0926>
  }]
 dyn_jacket_1_f_l [
  {
   shapeType=Capsule
   length=14.356
   radius=2.3457
   shapeName=dyn_jacket_1_f_l_capsule
   center=<8.3847 0.0185 3.4948>
   rotation=<rot -1.7555 94.1417 -104.016>
  }
  {shapeType=Capsule, length=13, radius=1.75, shapeName=dyn_jacket_1_f_l_capsule, center=<8.3339 0.5041 -1.866>, rotation=<rot -85 0.0002 0>}
 ]
 dyn_jacket_1_f_r [{
   shapeType=Capsule
   length=13.7598
   radius=2.3593
   shapeName=dyn_jacket_1_f_r_capsule
   center=<8.6563 0.715 -3.5992>
   rotation=<rot 1.2398 92.365 -68.0067>
  }]
 dyn_jacket_b [{
   shapeType=Capsule
   length=13.5884
   radius=5.6145
   shapeName=dyn_jacket_b_capsule
   center=<19.181 -0.7411 0.0413>
   rotation=<rot -0.4411 92.9903 0.5847>
  }]
 dyn_jacket_base_b [{shapeType=Capsule, length=0, radius=3.005, shapeName=dyn_jacket_base_b_capsule, rotation=<rot 0 -90 0>}]
 dyn_jacket_base_f_l [{shapeType=Capsule, length=0, radius=3.25, shapeName=dyn_jacket_base_f_l_capsule, rotation=<rot 0 -90 0>}]
 dyn_jacket_base_f_r [{shapeType=Capsule, length=0, radius=3.25, shapeName=dyn_jacket_base_f_r_capsule, rotation=<rot 0 -90 0>}]
 dyn_jacket_base_l [{shapeType=Capsule, length=0, radius=3.005, shapeName=dyn_jacket_base_l_capsule, rotation=<rot 0 -90 0>}]
 dyn_jacket_base_r [{shapeType=Capsule, length=0, radius=3.005, shapeName=dyn_jacket_base_r_capsule, rotation=<rot 0 -90 0>}]
 dyn_jacket_l [{
   shapeType=Capsule
   length=14.3088
   radius=4.8598
   shapeName=dyn_jacket_l_capsule
   center=<16.8445 0.098 3.9758>
   rotation=<rot 2.3232 148.5335 -75.5114>
  }]
 dyn_jacket_r [
  {
   shapeType=Capsule
   length=11.8525
   radius=6.0419
   shapeName=dyn_jacket_r_capsule
   center=<15.6661 -0.914 -4.4788>
   rotation=<rot -5.0266 -26.781 -76.0154>
  }
  {
   shapeType=Capsule
   length=15
   radius=2.5
   shapeName=dyn_jacket_r_capsule
   center=<10.0412 4.9368 -1.582>
   rotation=<rot -82.8125 -179.9977 179.9989>
  }
 ]
 pelvis [{shapeType=Capsule, length=7.4997, radius=9.6303, shapeName=pelvis_capsule, center=<-7.1972 -0.9744 0.326>, rotation=<rot 1.1371 -90 -1.6549>}]
 spine_01 [{
   shapeType=Capsule
   length=18.6019
   radius=7.2232
   shapeName=spine_01_capsule
   center=<-3.2866 -0.035 0.0731>
   rotation=<rot 0.9095 -90.0002 -0.2187>
  }]
 thigh_l [{
   shapeType=Capsule
   length=28.8182
   radius=7.3333
   shapeName=thigh_l_capsule
   center=<-21.9416 0.3129 1.2457>
   rotation=<rot -5.3415 88.851 -92.2771>
  }]
 thigh_r [{
   shapeType=Capsule
   length=29.0295
   radius=7.4753
   shapeName=thigh_r_capsule
   center=<21.7061 -0.6798 -1.4279>
   rotation=<rot -11.9007 89.0971 -88.0113>
  }]
}
constraints [
 {bone1Name=dyn_jacket_base_b, bone2Name=dyn_jacket_b, swing1LimitDegrees=40, swing2LimitDegrees=15, twistLimitDegrees=5}
 {bone1Name=dyn_jacket_base_l, bone2Name=dyn_jacket_l, swing1LimitDegrees=10, swing2LimitDegrees=20, twistLimitDegrees=2}
 {bone1Name=dyn_jacket_base_r, bone2Name=dyn_jacket_r, swing1LimitDegrees=10, swing2LimitDegrees=20, twistLimitDegrees=2}
 {
  bone1Name=dyn_jacket_base_r
  bone2Name=pelvis
  swing1LimitDegrees=45
  swing1Motion=Free
  swing2LimitDegrees=45
  swing2Motion=Free
  twistLimitDegrees=45
  twistMotion=Free
 }
 {
  bone1Name=dyn_jacket_base_b
  bone2Name=pelvis
  swing1LimitDegrees=45
  swing1Motion=Free
  swing2LimitDegrees=45
  swing2Motion=Free
  twistLimitDegrees=45
  twistMotion=Free
 }
 {
  bone1Name=dyn_jacket_base_l
  bone2Name=pelvis
  swing1LimitDegrees=45
  swing1Motion=Free
  swing2LimitDegrees=45
  swing2Motion=Free
  twistLimitDegrees=45
  twistMotion=Free
 }
 {bone1Name=dyn_jacket_1_f_l, bone2Name=dyn_jacket_base_f_l, swing1LimitDegrees=25, swing2LimitDegrees=15, twistLimitDegrees=5}
 {bone1Name=dyn_jacket_1_f_r, bone2Name=dyn_jacket_base_f_r, swing1LimitDegrees=25, swing2LimitDegrees=15, twistLimitDegrees=5}
]
