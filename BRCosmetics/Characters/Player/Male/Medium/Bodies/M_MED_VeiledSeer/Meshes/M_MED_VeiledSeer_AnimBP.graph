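# Animation Blueprint for the male medium Veiled Seer body cosmetic, adding secondary motion to the pose copied from its attached parent mesh.
# The copied pose feeds a Control Rig, then a pelvis-space RigidBody simulation with overridden gravity, then weighted constraints binding two button bones to spine and pelvis.
# Each update converts crouch and emote booleans into 0/1 values for the Control Rig.
# A second RigidBody node sits unconnected.
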AnimBlueprint /BRCosmetics/Characters/Player/Male/Medium/Bodies/M_MED_VeiledSeer/Meshes/M_MED_VeiledSeer_AnimBP
target_skeleton=&/M_MED_VeiledSeer_Skeleton
graphs [
 AnimGraph {nodes [
   RigidBody() {
    bOverrideWorldGravity=true
    overridePhysicsAsset=&/M_MED_VeiledSeer_Physics
    simulationSpace=BaseBoneSpace
    baseBoneRef {boneName=pelvis}
    componentAppliedLinearAccClamp=<3000 3000 3000>
    overrideWorldGravity {z=-980}
    simSpaceSettings {maxAngularAcceleration=75, maxAngularVelocity=5, maxLinearAcceleration=1750, maxLinearVelocity=100, velocityScaleZ=0.9, worldAlpha=0.9}
   }
   n1 = Control Rig!(
    Source=Copy Pose From Mesh() {bUseAttachedParent=true}
    Alpha
    is_crouch:double=Get Dyn_Is_Crouch() -> (Dyn_Is_Crouch:double)
    is_emote:double=Get Dyn_Is_Emote() -> (Dyn_Is_Emote:double)
   ) -> (Pose) {controlRigAssetReference {blueprintRigClass=&/M_MED_VeiledSeer_CtrlRig.M_MED_VeiledSeer_CtrlRig_C}}
   n2 = RigidBody(ComponentPose=Local To Component(LocalPose=@n1)) {
    bOverrideWorldGravity=true
    overridePhysicsAsset=&M_MED_VeiledSeer_Physics
    simulationSpace=BaseBoneSpace
    baseBoneRef {boneName=pelvis}
    componentAppliedLinearAccClamp=<2500 2500 2500>
    overrideWorldGravity {z=-980}
    simSpaceSettings {
     maxAngularAcceleration=100
     maxAngularVelocity=5
     maxLinearAcceleration=2500
     maxLinearVelocity=200
     velocityScaleZ=0.9
     worldAlpha=0.9
     externalLinearDragV=<1 1 1>
    }
   }
   n3 = "Constraint - dyn_button_1 (PPPPPP)"!(
    ComponentPose=@n2
    Alpha=1
    ConstraintWeights_0=0.3655
    ConstraintWeights_1:float=0.3058
    ConstraintWeights_2:float=0.2227
    ConstraintWeights_3:float=0.0598
    ConstraintWeights_4:float=0.0301
    ConstraintWeights_5:float=0.0162
   ) -> (Pose) {
    boneToModify {boneName=dyn_button_1}
    constraintSetup [
     {targetBone {boneName=spine_03}}
     {targetBone {boneName=spine_02}}
     {targetBone {boneName=spine_04}}
     {targetBone {boneName=spine_01}}
     {targetBone {boneName=spine_05}}
     {targetBone {boneName=pelvis}}
    ]
    constraintWeights [1,1,1,1,1,1]
   }
   n4 = "Constraint - dyn_button_2 (PPPPP)"!(
    ComponentPose=@n3
    Alpha=1
    ConstraintWeights_0=0.3405
    ConstraintWeights_1:float=0.3009
    ConstraintWeights_2:float=0.1761
    ConstraintWeights_3:float=0.0928
    ConstraintWeights_4:float=0.0898
   ) -> (Pose) {
    boneToModify {boneName=dyn_button_2}
    constraintSetup [
     {targetBone {boneName=spine_02}}
     {targetBone {boneName=spine_03}}
     {targetBone {boneName=spine_01}}
     {targetBone {boneName=spine_04}}
     {targetBone {boneName=pelvis}}
    ]
    constraintWeights [1,1,1,1,1]
   }
   Output Pose(Result=Component To Local(ComponentPose=@n4)) {layerGroup=None}
  ]}
 EventGraph {nodes [
   Event BlueprintUpdateAnimation()
   > n1 = Sequence()
   > n2 = Branch(Condition=Get bIsCrouching())
   > Set Dyn_Is_Crouch(Dyn_Is_Crouch:double=1) -> (Output_Get:double)
   Set Dyn_Is_Crouch(execute=@n2.else, Dyn_Is_Crouch=0)
   n3 = Branch(execute=@n1.then_1, Condition=Get bIsPlayingEmote())
   > Set Dyn_Is_Emote(Dyn_Is_Emote:double=1) -> (Output_Get:double)
   Set Dyn_Is_Emote(execute=@n3.else, Dyn_Is_Emote=0)
  ]}
]
variables [
 Dyn_Is_Crouch {type=number}
 Dyn_Is_Emote {type=number}
]
