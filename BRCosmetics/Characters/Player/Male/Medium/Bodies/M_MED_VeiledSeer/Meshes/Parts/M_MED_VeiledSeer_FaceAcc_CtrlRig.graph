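# Control Rig for the Veiled Seer face accessory that applies pose-driven corrective offsets to its left, back and right dyn_hair base bones.
# Sequenced spherical pose readers on head, clavicles and pelvis weight Modify Transforms nodes, with Dyn_crouch, Dyn_skydive and Dyn_FE scaling some offsets.
# One head reader also stores Dyn_Rotate_l, multiplied by Dyn_FE to weight a final back/left correction.
# The neck_01 reader branch is disconnected from execution.
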ControlRigBlueprint /BRCosmetics/Characters/Player/Male/Medium/Bodies/M_MED_VeiledSeer/Meshes/Parts/M_MED_VeiledSeer_FaceAcc_CtrlRig
graphs [
 RigVMFunctionLibrary {links [], nodes []}
 RigVMFunction_Sequence_3_ContainedGraph {nodes [
   n1 = RigVMFunctionEntryNode()
   RigVMFunction_Sequence = &RigVMFunction_Sequence(ExecuteContext=@n1.ExecuteContext)
   n2 = &RigVMFunction_Sequence(ExecuteContext=@RigVMFunction_Sequence.B)
   n3 = &RigVMFunction_Sequence(ExecuteContext=@n2.B)
   n4 = &RigVMFunction_Sequence(ExecuteContext=@n3.B)
   n5 = &RigVMFunction_Sequence(ExecuteContext=@n4.B)
   RigVMFunctionReturnNode(A=@RigVMFunction_Sequence.A, B=@n2.A, C=@n3.A, D=@n4.A, E=@n5.A, F=@n5.B)
  ]}
 RigVMFunction_Sequence_4_ContainedGraph {nodes [
   n1 = RigVMFunctionEntryNode()
   n2 = &RigVMFunction_Sequence(ExecuteContext=@n1.ExecuteContext)
   n3 = &RigVMFunction_Sequence(ExecuteContext=@n2.B)
   RigVMFunctionReturnNode(B=@n3.A, C=@n3.B, A=@n2.A)
  ]}
 RigVMFunction_Sequence_4_ContainedGraph {nodes [
   n1 = RigVMFunctionEntryNode()
   n2 = &RigVMFunction_Sequence(ExecuteContext=@n1.ExecuteContext)
   n3 = &RigVMFunction_Sequence(ExecuteContext=@n2.B)
   RigVMFunctionReturnNode(B=@n3.A, C=@n3.B, A=@n2.A)
  ]}
 RigVMModel {nodes [
   n1 = &RigUnit_SphericalPoseReader(
    ActiveRegionScaleFactors=(PositiveWidth=0.1,NegativeWidth=0.1,PositiveHeight=0.1,NegativeHeight=0.1)
    ActiveRegionSize=0.1
    Debug=(bDrawDebug=true,bDraw2D=False,bDrawLocalAxes=False,DebugScale=25,DebugSegments=20,DebugThickness=0.25)
    DriverAxis=(X=1,Y=0,Z=0)
    DriverItem=(Type=Bone,Name="neck_01")
    ExecutePin=()
    FalloffRegionScaleFactors=(PositiveWidth=1,NegativeWidth=1,PositiveHeight=1,NegativeHeight=1)
    FalloffSize=0.05
    FlipHeightScaling=False
    FlipWidthScaling=False
    OptionalParentItem=(Type=Bone,Name="None")
    RotationOffset=(X=-70,Y=-40,Z=47.5)
   )
   Multiply = &RigVMFunction_MathDoubleMul(B=1, A=@n1.OutputParam)
   &RigUnit_ModifyTransforms(
    ExecutePin=()
    ItemToModify=((Item=(Type=Bone,Name="dyn_hair_base_l"),Transform=(Rotation=(X=0,Y=0,Z=0,W=1),Translation=(X=0,Y=0,Z=0),Scale3D=(X=1,Y=1,Z=1))))
    WeightMinimum=0
    ItemToModify.0.Transform.Rotation=&RigVMFunction_MathQuaternionFromEuler(Euler=(X=0,Y=0,Z=0)).Result
    Weight=@Multiply.Result
   )
   n2 = RigVMVariableNode() {variable=Dyn_crouch}
   n3 = &RigVMFunction_MathDoubleMul(A=0, B=@n2.Value)
   n4 = &RigVMFunction_MathDoubleMul(A=0, B=@n2.Value)
   n5 = &RigVMFunction_MathDoubleMul(A=0, B=@n2.Value)
   > &RigUnit_ModifyTransforms(
    ItemToModify=((Item=(Type=Bone,Name="dyn_hair_base_b"),Transform=(Rotation=(X=0,Y=0,Z=0,W=1),Translation=(X=-5,Y=2,Z=5),Scale3D=(X=1,Y=1,Z=1))))
    WeightMinimum=0
    ItemToModify.0.Transform.Translation.Y=&RigVMFunction_MathDoubleAdd(A=0, B=&RigVMFunction_MathDoubleMul(A=0, B=@n2.Value).Result).Result
    ItemToModify.0.Transform.Translation.Z=&RigVMFunction_MathDoubleAdd(A=0, B=&RigVMFunction_MathDoubleMul(A=0, B=@n2.Value).Result).Result
    ItemToModify.0.Transform.Translation.X=&RigVMFunction_MathDoubleAdd(A=0, B=&RigVMFunction_MathDoubleMul(A=0, B=@n2.Value).Result).Result
    ItemToModify.0.Transform.Rotation=&RigVMFunction_MathQuaternionFromEuler(
     Euler=(X=-20,Y=40,Z=-60)
     Euler.Y=&RigVMFunction_MathDoubleAdd(A=0, B=@n3.Result).Result
     Euler.Z=&RigVMFunction_MathDoubleAdd(A=0, B=@n4.Result).Result
     Euler.X=&RigVMFunction_MathDoubleAdd(A=0, B=@n5.Result).Result
    ).Result
    Weight=@Multiply.Result
   )
   n6 = RigVMVariableNode() {variable=Dyn_crouch}
   n7 = &RigVMFunction_MathDoubleMul(A=0, B=@n6.Value)
   n8 = &RigVMFunction_MathDoubleMul(A=0, B=@n6.Value)
   n9 = &RigVMFunction_MathDoubleMul(A=0, B=@n6.Value)
   > &RigUnit_ModifyTransforms(
    ItemToModify=((Item=(Type=Bone,Name="dyn_hair_base_r"),Transform=(Rotation=(X=0,Y=0,Z=0,W=1),Translation=(X=-5,Y=0,Z=-3),Scale3D=(X=1,Y=1,Z=1))))
    WeightMinimum=0
    ItemToModify.0.Transform.Translation.X=&RigVMFunction_MathDoubleAdd(A=0, B=&RigVMFunction_MathDoubleMul(B=0, A=@n6.Value).Result).Result
    ItemToModify.0.Transform.Translation.Y=&RigVMFunction_MathDoubleAdd(A=0, B=&RigVMFunction_MathDoubleMul(A=0, B=@n6.Value).Result).Result
    ItemToModify.0.Transform.Translation.Z=&RigVMFunction_MathDoubleAdd(A=0, B=&RigVMFunction_MathDoubleMul(A=0, B=@n6.Value).Result).Result
    ItemToModify.0.Transform.Rotation=&RigVMFunction_MathQuaternionFromEuler(
     Euler=(X=0,Y=0,Z=0)
     Euler.Y=&RigVMFunction_MathDoubleAdd(A=0, B=@n7.Result).Result
     Euler.Z=&RigVMFunction_MathDoubleAdd(A=0, B=@n8.Result).Result
     Euler.X=&RigVMFunction_MathDoubleAdd(A=0, B=@n9.Result).Result
    ).Result
    Weight=@Multiply.Result
   )
   RigUnit_BeginExecution = &RigUnit_BeginExecution()
   n10 = &RigVMFunction_Sequence(ExecuteContext=@RigUnit_BeginExecution.ExecutePin)
   RigVMFunction_Sequence = RigVMAggregateNode(ExecuteContext=@n10.A)
   n11 = &RigUnit_SphericalPoseReader(
    ActiveRegionScaleFactors=(PositiveWidth=1,NegativeWidth=3,PositiveHeight=0.5,NegativeHeight=1)
    ActiveRegionSize=0.1
    Debug=(bDrawDebug=true,bDraw2D=False,bDrawLocalAxes=False,DebugScale=25,DebugSegments=20,DebugThickness=0.25)
    DriverAxis=(X=1,Y=0,Z=0)
    DriverItem=(Type=Bone,Name="head")
    FalloffRegionScaleFactors=(PositiveWidth=1,NegativeWidth=1,PositiveHeight=1,NegativeHeight=1)
    FalloffSize=0.35
    FlipHeightScaling=False
    FlipWidthScaling=False
    OptionalParentItem=(Type=Bone,Name="None")
    RotationOffset=(X=-70,Y=-115,Z=15)
    ExecutePin=@RigVMFunction_Sequence.B
   )
   > &RigUnit_ModifyTransforms(
    ItemToModify=((Item=(Type=Bone,Name="dyn_hair_base_l"),Transform=(Rotation=(X=0,Y=0,Z=0,W=1),Translation=(X=-1,Y=1.5,Z=0),Scale3D=(X=1,Y=1,Z=1))))
    WeightMinimum=0
    ItemToModify.0.Transform.Rotation=&RigVMFunction_MathQuaternionFromEuler(Euler=(X=0,Y=0,Z=20)).Result
    Weight=@n11.OutputParam
   )
   n12 = RigVMVariableNode() {variable=Dyn_crouch}
   n13 = &RigVMFunction_MathDoubleMul(A=-45, B=@n12.Value)
   n14 = &RigVMFunction_MathDoubleMul(A=0.02, B=@n12.Value)
   n15 = &RigVMFunction_MathDoubleMul(A=15, B=@n12.Value)
   > &RigUnit_ModifyTransforms(
    ItemToModify=((Item=(Type=Bone,Name="dyn_hair_base_b"),Transform=(Rotation=(X=0,Y=0,Z=0,W=1),Translation=(X=-5,Y=2,Z=5),Scale3D=(X=1,Y=1,Z=1))))
    WeightMinimum=0
    ItemToModify.0.Transform.Translation.Y=&RigVMFunction_MathDoubleAdd(A=0, B=&RigVMFunction_MathDoubleMul(A=0, B=@n12.Value).Result).Result
    ItemToModify.0.Transform.Translation.Z=&RigVMFunction_MathDoubleAdd(A=3.5, B=&RigVMFunction_MathDoubleMul(A=2, B=@n12.Value).Result).Result
    ItemToModify.0.Transform.Translation.X=&RigVMFunction_MathDoubleAdd(A=0, B=&RigVMFunction_MathDoubleMul(A=0, B=@n12.Value).Result).Result
    ItemToModify.0.Transform.Rotation=&RigVMFunction_MathQuaternionFromEuler(
     Euler=(X=-20,Y=40,Z=-60)
     Euler.Y=&RigVMFunction_MathDoubleAdd(A=-30, B=@n13.Result).Result
     Euler.Z=&RigVMFunction_MathDoubleAdd(A=0, B=@n14.Result).Result
     Euler.X=&RigVMFunction_MathDoubleAdd(A=0, B=@n15.Result).Result
    ).Result
    Weight=@n11.OutputParam
   )
   n16 = RigVMVariableNode() {variable=Dyn_crouch}
   n17 = &RigVMFunction_MathDoubleMul(A=0, B=@n16.Value)
   n18 = &RigVMFunction_MathDoubleMul(A=-30, B=@n16.Value)
   n19 = &RigVMFunction_MathDoubleMul(A=0, B=@n16.Value)
   > &RigUnit_ModifyTransforms(
    ItemToModify=((Item=(Type=Bone,Name="dyn_hair_base_r"),Transform=(Rotation=(X=0,Y=0,Z=0,W=1),Translation=(X=-5,Y=0,Z=-3),Scale3D=(X=1,Y=1,Z=1))))
    WeightMinimum=0
    ItemToModify.0.Transform.Translation.Z=&RigVMFunction_MathDoubleAdd(A=0.5, B=&RigVMFunction_MathDoubleMul(A=2, B=@n16.Value).Result).Result
    ItemToModify.0.Transform.Translation.X=&RigVMFunction_MathDoubleAdd(A=-1.5, B=&RigVMFunction_MathDoubleMul(A=0, B=@n16.Value).Result).Result
    ItemToModify.0.Transform.Translation.Y=&RigVMFunction_MathDoubleAdd(A=0.5, B=&RigVMFunction_MathDoubleMul(A=-3, B=@n16.Value).Result).Result
    ItemToModify.0.Transform.Rotation=&RigVMFunction_MathQuaternionFromEuler(
     Euler=(X=-20,Y=40,Z=0)
     Euler.Y=&RigVMFunction_MathDoubleAdd(A=0, B=@n17.Result).Result
     Euler.Z=&RigVMFunction_MathDoubleAdd(A=-20, B=@n18.Result).Result
     Euler.X=&RigVMFunction_MathDoubleAdd(A=0, B=@n19.Result).Result
    ).Result
    Weight=@n11.OutputParam
   )
   n20 = &RigUnit_SphericalPoseReader(
    ActiveRegionScaleFactors=(PositiveWidth=0.5,NegativeWidth=0.5,PositiveHeight=0.5,NegativeHeight=0.5)
    ActiveRegionSize=0.1
    Debug=(bDrawDebug=false,bDraw2D=False,bDrawLocalAxes=False,DebugScale=25,DebugSegments=20,DebugThickness=0.25)
    DriverAxis=(X=1,Y=0,Z=0)
    DriverItem=(Type=Bone,Name="head")
    FalloffRegionScaleFactors=(PositiveWidth=1,NegativeWidth=1,PositiveHeight=1,NegativeHeight=1)
    FalloffSize=0.25
    FlipHeightScaling=False
    FlipWidthScaling=False
    OptionalParentItem=(Type=Bone,Name="None")
    RotationOffset=(X=90,Y=0,Z=-65)
    ExecutePin=@RigVMFunction_Sequence.C
   )
   n21 = RigVMVariableNode() {variable=Dyn_crouch}
   n22 = &RigVMFunction_MathDoubleMul(A=0, B=@n21.Value)
   n23 = &RigVMFunction_MathDoubleMul(A=0, B=@n21.Value)
   n24 = &RigVMFunction_MathDoubleMul(A=0, B=@n21.Value)
   > &RigUnit_ModifyTransforms(
    ItemToModify=((Item=(Type=Bone,Name="dyn_hair_base_l"),Transform=(Rotation=(X=0,Y=0,Z=0,W=1),Translation=(X=0,Y=-2,Z=0),Scale3D=(X=1,Y=1,Z=1))))
    WeightMinimum=0
    ItemToModify.0.Transform.Translation.Z=&RigVMFunction_MathDoubleAdd(A=0, B=&RigVMFunction_MathDoubleMul(A=0, B=@n21.Value).Result).Result
    ItemToModify.0.Transform.Translation.X=&RigVMFunction_MathDoubleAdd(A=1, B=&RigVMFunction_MathDoubleMul(A=0, B=@n21.Value).Result).Result
    ItemToModify.0.Transform.Translation.Y=&RigVMFunction_MathDoubleAdd(A=0.5, B=&RigVMFunction_MathDoubleMul(A=0, B=@n21.Value).Result).Result
    ItemToModify.0.Transform.Rotation=&RigVMFunction_MathQuaternionFromEuler(
     Euler=(X=0,Y=-10,Z=75)
     Euler.Z=&RigVMFunction_MathDoubleAdd(A=10, B=@n22.Result).Result
     Euler.X=&RigVMFunction_MathDoubleAdd(A=0, B=@n23.Result).Result
     Euler.Y=&RigVMFunction_MathDoubleAdd(A=20, B=@n24.Result).Result
    ).Result
    Weight=@n20.OutputParam
   )
   n25 = RigVMVariableNode() {variable=Dyn_crouch}
   n26 = &RigVMFunction_MathDoubleMul(A=0, B=@n25.Value)
   n27 = &RigVMFunction_MathDoubleMul(A=0, B=@n25.Value)
   n28 = &RigVMFunction_MathDoubleMul(A=0, B=@n25.Value)
   > &RigUnit_ModifyTransforms(
    ItemToModify=((Item=(Type=Bone,Name="dyn_hair_base_b"),Transform=(Rotation=(X=0,Y=0,Z=0,W=1),Translation=(X=0,Y=0,Z=0),Scale3D=(X=1,Y=1,Z=1))))
    WeightMinimum=0
    ItemToModify.0.Transform.Rotation=&RigVMFunction_MathQuaternionFromEuler(
     Euler=(X=-20,Y=-30,Z=-30)
     Euler.Y=&RigVMFunction_MathDoubleAdd(A=0, B=@n26.Result).Result
     Euler.Z=&RigVMFunction_MathDoubleAdd(A=0, B=@n27.Result).Result
     Euler.X=&RigVMFunction_MathDoubleAdd(A=0, B=@n28.Result).Result
    ).Result
    Weight=@n20.OutputParam
   )
   n29 = RigVMVariableNode() {variable=Dyn_crouch}
   n30 = &RigVMFunction_MathDoubleMul(A=0, B=@n29.Value)
   n31 = &RigVMFunction_MathDoubleMul(A=0, B=@n29.Value)
   n32 = &RigVMFunction_MathDoubleMul(A=0, B=@n29.Value)
   > &RigUnit_ModifyTransforms(
    ItemToModify=((Item=(Type=Bone,Name="dyn_hair_base_r"),Transform=(Rotation=(X=0,Y=0,Z=0,W=1),Translation=(X=0,Y=0,Z=0),Scale3D=(X=1,Y=1,Z=1))))
    WeightMinimum=0
    ItemToModify.0.Transform.Rotation=&RigVMFunction_MathQuaternionFromEuler(
     Euler=(X=-60,Y=0,Z=30)
     Euler.X=&RigVMFunction_MathDoubleAdd(A=0, B=@n30.Result).Result
     Euler.Z=&RigVMFunction_MathDoubleAdd(A=0, B=@n31.Result).Result
     Euler.Y=&RigVMFunction_MathDoubleAdd(A=0, B=@n32.Result).Result
    ).Result
    Weight=@n20.OutputParam
   )
   n33 = &RigVMFunction_Sequence(ExecuteContext=@RigVMFunction_Sequence.D)
   n34 = &RigUnit_SphericalPoseReader(
    ActiveRegionScaleFactors=(PositiveWidth=1.5,NegativeWidth=1.5,PositiveHeight=1,NegativeHeight=0.5)
    ActiveRegionSize=0.1
    Debug=(bDrawDebug=true,bDraw2D=False,bDrawLocalAxes=False,DebugScale=25,DebugSegments=20,DebugThickness=0.25)
    DriverAxis=(X=0,Y=1,Z=0)
    DriverItem=(Type=Bone,Name="head")
    FalloffRegionScaleFactors=(PositiveWidth=1,NegativeWidth=1,PositiveHeight=1,NegativeHeight=1)
    FalloffSize=0.2
    FlipHeightScaling=False
    FlipWidthScaling=False
    OptionalParentItem=(Type=Bone,Name="None")
    RotationOffset=(X=120,Y=0,Z=-10)
    ExecutePin=@n33.A
   )
   > RigVMVariableNode(Value=@n34.OutputParam) {variable=Dyn_Rotate_l}
   Add = &RigVMFunction_MathDoubleAdd(A=0, B=0)
   n35 = RigVMVariableNode() {variable=Dyn_FE}
   n36 = &RigVMFunction_MathDoubleMul(B=10, A=@n35.Value)
   n37 = &RigVMFunction_MathDoubleMul(B=10, A=@n35.Value)
   > &RigUnit_ModifyTransforms(
    ItemToModify=((Item=(Type=Bone,Name="dyn_hair_base_l"),Transform=(Rotation=(X=0,Y=0,Z=0,W=1),Translation=(X=-1,Y=0,Z=0),Scale3D=(X=1,Y=1,Z=1))))
    WeightMinimum=0
    ItemToModify.0.Transform.Translation.Y=@Add.Result
    ItemToModify.0.Transform.Translation.X=&RigVMFunction_MathDoubleAdd(A=1, B=&RigVMFunction_MathDoubleMul(A=1, B=-1).Result).Result
    ItemToModify.0.Transform.Translation.Z=&RigVMFunction_MathDoubleAdd(A=0, B=&RigVMFunction_MathDoubleMul(B=0.5, A=@n35.Value).Result).Result
    ItemToModify.0.Transform.Rotation=&RigVMFunction_MathQuaternionFromEuler(
     Euler=(X=0,Y=-20,Z=0)
     Euler.Z=&RigVMFunction_MathDoubleAdd(A=20, B=0).Result
     Euler.X=&RigVMFunction_MathDoubleAdd(A=-20, B=@n36.Result).Result
     Euler.Y=&RigVMFunction_MathDoubleAdd(A=-45, B=@n37.Result).Result
    ).Result
    Weight=@n34.OutputParam
   )
   n38 = RigVMVariableNode() {variable=Dyn_FE}
   n39 = &RigVMFunction_MathDoubleMul(B=60, A=@n38.Value)
   n40 = &RigVMFunction_MathDoubleMul(B=15, A=@n38.Value)
   n41 = &RigVMFunction_MathDoubleMul(B=45, A=@n38.Value)
   > &RigUnit_ModifyTransforms(
    ItemToModify=((Item=(Type=Bone,Name="dyn_hair_base_b"),Transform=(Rotation=(X=0,Y=0,Z=0,W=1),Translation=(X=-1,Y=-2,Z=0),Scale3D=(X=1,Y=1,Z=1))))
    WeightMinimum=0
    ItemToModify.0.Transform.Translation.Z=&RigVMFunction_MathDoubleAdd(A=0, B=0).Result
    ItemToModify.0.Transform.Translation.X=&RigVMFunction_MathDoubleAdd(A=-1, B=0).Result
    ItemToModify.0.Transform.Translation.Y=&RigVMFunction_MathDoubleAdd(A=0, B=&RigVMFunction_MathDoubleMul(B=3, A=@n38.Value).Result).Result
    ItemToModify.0.Transform.Rotation=&RigVMFunction_MathQuaternionFromEuler(
     Euler=(X=60,Y=-10,Z=-30)
     Euler.Z=&RigVMFunction_MathDoubleAdd(A=0, B=@n39.Result).Result
     Euler.X=&RigVMFunction_MathDoubleAdd(A=0, B=@n40.Result).Result
     Euler.Y=&RigVMFunction_MathDoubleAdd(A=-35, B=@n41.Result).Result
    ).Result
    Weight=@n34.OutputParam
   )
   n42 = RigVMVariableNode() {variable=Dyn_FE}
   n43 = &RigVMFunction_MathDoubleMul(B=0, A=@n42.Value)
   n44 = &RigVMFunction_MathDoubleMul(B=0, A=@n42.Value)
   n45 = &RigVMFunction_MathDoubleMul(B=-30, A=@n42.Value)
   > &RigUnit_ModifyTransforms(
    ItemToModify=((Item=(Type=Bone,Name="dyn_hair_base_r"),Transform=(Rotation=(X=0,Y=0,Z=0,W=1),Translation=(X=0,Y=0,Z=1.5),Scale3D=(X=1,Y=1,Z=1))))
    WeightMinimum=0
    ItemToModify.0.Transform.Translation.Z=&RigVMFunction_MathDoubleAdd(A=0, B=&RigVMFunction_MathDoubleMul(B=-1, A=@n42.Value).Result).Result
    ItemToModify.0.Transform.Translation.Y=&RigVMFunction_MathDoubleAdd(A=-1, B=&RigVMFunction_MathDoubleMul(B=-1, A=@n42.Value).Result).Result
    ItemToModify.0.Transform.Translation.X=&RigVMFunction_MathDoubleAdd(A=-1, B=&RigVMFunction_MathDoubleMul(B=1, A=@n42.Value).Result).Result
    ItemToModify.0.Transform.Rotation=&RigVMFunction_MathQuaternionFromEuler(
     Euler=(X=-10,Y=15,Z=-30)
     Euler.X=&RigVMFunction_MathDoubleAdd(A=0, B=@n43.Result).Result
     Euler.Y=&RigVMFunction_MathDoubleAdd(A=-10, B=@n44.Result).Result
     Euler.Z=&RigVMFunction_MathDoubleAdd(A=0, B=@n45.Result).Result
    ).Result
    Weight=@n34.OutputParam
   )
   n46 = &RigUnit_SphericalPoseReader(
    ActiveRegionScaleFactors=(PositiveWidth=1.5,NegativeWidth=1.5,PositiveHeight=0.15,NegativeHeight=1)
    ActiveRegionSize=0.1
    Debug=(bDrawDebug=true,bDraw2D=False,bDrawLocalAxes=False,DebugScale=25,DebugSegments=20,DebugThickness=0.25)
    DriverAxis=(X=0,Y=1,Z=0)
    DriverItem=(Type=Bone,Name="head")
    FalloffRegionScaleFactors=(PositiveWidth=1,NegativeWidth=1,PositiveHeight=1,NegativeHeight=1)
    FalloffSize=0.2
    FlipHeightScaling=False
    FlipWidthScaling=False
    OptionalParentItem=(Type=Bone,Name="None")
    RotationOffset=(X=60,Y=0,Z=-10)
    ExecutePin=@n33.B
   )
   > &RigUnit_ModifyTransforms(
    ItemToModify=((Item=(Type=Bone,Name="dyn_hair_base_l"),Transform=(Rotation=(X=0,Y=0,Z=0,W=1),Translation=(X=-1,Y=0,Z=0),Scale3D=(X=1,Y=1,Z=1))))
    WeightMinimum=0
    ItemToModify.0.Transform.Translation.Y=&RigVMFunction_MathDoubleAdd(A=2, B=0).Result
    ItemToModify.0.Transform.Translation.X=&RigVMFunction_MathDoubleAdd(A=-2, B=2).Result
    ItemToModify.0.Transform.Rotation=&RigVMFunction_MathQuaternionFromEuler(
     Euler=(X=0,Y=-20,Z=0)
     Euler.Z=&RigVMFunction_MathDoubleAdd(A=-30, B=0).Result
     Euler.X=&RigVMFunction_MathDoubleAdd(A=30, B=0).Result
     Euler.Y=&RigVMFunction_MathDoubleAdd(A=-10, B=0).Result
    ).Result
    Weight=@n46.OutputParam
   )
   > &RigUnit_ModifyTransforms(
    ItemToModify=((Item=(Type=Bone,Name="dyn_hair_base_b"),Transform=(Rotation=(X=0,Y=0,Z=0,W=1),Translation=(X=-1,Y=-2,Z=0),Scale3D=(X=1,Y=1,Z=1))))
    WeightMinimum=0
    ItemToModify.0.Transform.Translation.Z=&RigVMFunction_MathDoubleAdd(A=2, B=0).Result
    ItemToModify.0.Transform.Translation.X=&RigVMFunction_MathDoubleAdd(A=-1, B=0).Result
    ItemToModify.0.Transform.Translation.Y=&RigVMFunction_MathDoubleAdd(A=0, B=0).Result
    ItemToModify.0.Transform.Rotation=&RigVMFunction_MathQuaternionFromEuler(
     Euler=(X=60,Y=-10,Z=-30)
     Euler.Z=&RigVMFunction_MathDoubleAdd(A=-20, B=0).Result
     Euler.X=&RigVMFunction_MathDoubleAdd(A=0, B=0).Result
     Euler.Y=&RigVMFunction_MathDoubleAdd(A=0, B=-20).Result
    ).Result
    Weight=@n46.OutputParam
   )
   n47 = &RigVMFunction_MathDoubleMul(A=45, B=RigVMVariableNode() {variable=Dyn_crouch}.Value)
   > &RigUnit_ModifyTransforms(
    ItemToModify=((Item=(Type=Bone,Name="dyn_hair_base_r"),Transform=(Rotation=(X=0,Y=0,Z=0,W=1),Translation=(X=0,Y=0,Z=1.5),Scale3D=(X=1,Y=1,Z=1))))
    WeightMinimum=0
    ItemToModify.0.Transform.Translation.X=&RigVMFunction_MathDoubleAdd(A=-1, B=0).Result
    ItemToModify.0.Transform.Translation.Y=&RigVMFunction_MathDoubleAdd(A=0, B=0).Result
    ItemToModify.0.Transform.Translation.Z=&RigVMFunction_MathDoubleAdd(A=0, B=0).Result
    ItemToModify.0.Transform.Rotation=&RigVMFunction_MathQuaternionFromEuler(
     Euler=(X=-10,Y=15,Z=-30)
     Euler.X=&RigVMFunction_MathDoubleAdd(A=0, B=0).Result
     Euler.Y=&RigVMFunction_MathDoubleAdd(A=0, B=0).Result
     Euler.Z=&RigVMFunction_MathDoubleAdd(A=30, B=@n47.Result).Result
    ).Result
    Weight=@n46.OutputParam
   )
   RigVMFunction_Sequence_3 = RigVMAggregateNode(ExecuteContext=@RigVMFunction_Sequence.E)
   n48 = &RigUnit_SphericalPoseReader(
    ActiveRegionScaleFactors=(PositiveWidth=0.5,NegativeWidth=0.5,PositiveHeight=0.5,NegativeHeight=0.5)
    ActiveRegionSize=0.1
    Debug=(bDrawDebug=true,bDraw2D=False,bDrawLocalAxes=False,DebugScale=25,DebugSegments=20,DebugThickness=0.25)
    DriverAxis=(X=1,Y=0,Z=0)
    DriverItem=(Type=Bone,Name="clavicle_l")
    FalloffRegionScaleFactors=(PositiveWidth=1,NegativeWidth=1,PositiveHeight=1,NegativeHeight=1)
    FalloffSize=0.3
    FlipHeightScaling=False
    FlipWidthScaling=False
    OptionalParentItem=(Type=Bone,Name="None")
    RotationOffset=(X=0,Y=-45,Z=0)
    ExecutePin=@RigVMFunction_Sequence_3.A
   )
   n49 = RigVMVariableNode() {variable=Dyn_skydive}
   n50 = &RigVMFunction_AlphaInterp(InterpSpeedDecreasing=50, InterpSpeedIncreasing=50, Scale=1, bClampResult=False, bInterpResult=False, Value=@n48.OutputParam)
   > &RigUnit_ModifyTransforms(
    ItemToModify=((Item=(Type=Bone,Name="dyn_hair_base_l"),Transform=(Rotation=(X=0,Y=0,Z=0,W=1),Translation=(X=-1,Y=2,Z=0),Scale3D=(X=1,Y=1,Z=1))))
    WeightMinimum=0
    ItemToModify.0.Transform.Translation.Y=&RigVMFunction_MathDoubleAdd(A=0, B=&RigVMFunction_MathDoubleMul(A=0, B=@n49.Value).Result).Result
    ItemToModify.0.Transform.Translation.X=&RigVMFunction_MathDoubleAdd(A=-1, B=&RigVMFunction_MathDoubleMul(A=-0.5, B=@n49.Value).Result).Result
    ItemToModify.0.Transform.Rotation=&RigVMFunction_MathQuaternionFromEuler(Euler=(X=-20,Y=10,Z=10)).Result
    Weight=@n50.Result
   )
   > &RigUnit_ModifyTransforms(
    ItemToModify=((Item=(Type=Bone,Name="dyn_hair_base_b"),Transform=(Rotation=(X=0,Y=0,Z=0,W=1),Translation=(X=-1,Y=0,Z=0),Scale3D=(X=1,Y=1,Z=1))))
    WeightMinimum=0
    ItemToModify.0.Transform.Rotation=&RigVMFunction_MathQuaternionFromEuler(Euler=(X=0,Y=0,Z=0)).Result
    Weight=@n50.Result
   )
   n51 = &RigUnit_SphericalPoseReader(
    ActiveRegionScaleFactors=(PositiveWidth=0.5,NegativeWidth=0.5,PositiveHeight=0.5,NegativeHeight=0.5)
    ActiveRegionSize=0.1
    Debug=(bDrawDebug=true,bDraw2D=False,bDrawLocalAxes=False,DebugScale=25,DebugSegments=20,DebugThickness=0.25)
    DriverAxis=(X=-1,Y=0,Z=0)
    DriverItem=(Type=Bone,Name="clavicle_r")
    FalloffRegionScaleFactors=(PositiveWidth=1,NegativeWidth=1,PositiveHeight=1,NegativeHeight=1)
    FalloffSize=0.3
    FlipHeightScaling=False
    FlipWidthScaling=False
    OptionalParentItem=(Type=Bone,Name="None")
    RotationOffset=(X=-180,Y=-45,Z=0)
    ExecutePin=@RigVMFunction_Sequence_3.B
   )
   n52 = RigVMVariableNode() {variable=Dyn_skydive}
   AlphaInterp = &RigVMFunction_AlphaInterp(InterpSpeedDecreasing=50, InterpSpeedIncreasing=50, Scale=1, bClampResult=False, bInterpResult=False, Value=@n51.OutputParam)
   > &RigUnit_ModifyTransforms(
    ItemToModify=((Item=(Type=Bone,Name="dyn_hair_base_r"),Transform=(Rotation=(X=0,Y=0,Z=0,W=1),Translation=(X=-1,Y=2,Z=0),Scale3D=(X=1,Y=1,Z=1))))
    WeightMinimum=0
    ItemToModify.0.Transform.Translation.Y=&RigVMFunction_MathDoubleAdd(A=2, B=&RigVMFunction_MathDoubleMul(A=-1, B=@n52.Value).Result).Result
    ItemToModify.0.Transform.Rotation=&RigVMFunction_MathQuaternionFromEuler(Euler=(X=20,Y=0,Z=10)).Result
    Weight=@AlphaInterp.Result
   )
   > &RigUnit_ModifyTransforms(
    ItemToModify=((Item=(Type=Bone,Name="dyn_hair_base_b"),Transform=(Rotation=(X=0,Y=0,Z=0,W=1),Translation=(X=-1,Y=0,Z=0),Scale3D=(X=1,Y=1,Z=1))))
    WeightMinimum=0
    ItemToModify.0.Transform.Rotation=&RigVMFunction_MathQuaternionFromEuler(Euler=(X=0,Y=0,Z=0)).Result
    Weight=@AlphaInterp.Result
   )
   RigVMFunction_Sequence_4 = RigVMAggregateNode(ExecuteContext=@RigVMFunction_Sequence_3.C)
   n53 = &RigUnit_SphericalPoseReader(
    ActiveRegionScaleFactors=(PositiveWidth=0.5,NegativeWidth=0.5,PositiveHeight=0.5,NegativeHeight=0.5)
    ActiveRegionSize=0.1
    Debug=(bDrawDebug=true,bDraw2D=False,bDrawLocalAxes=False,DebugScale=25,DebugSegments=20,DebugThickness=0.25)
    DriverAxis=(X=0,Y=1,Z=0)
    DriverItem=(Type=Bone,Name="pelvis")
    FalloffRegionScaleFactors=(PositiveWidth=1,NegativeWidth=1,PositiveHeight=1,NegativeHeight=1)
    FalloffSize=0.25
    FlipHeightScaling=False
    FlipWidthScaling=False
    OptionalParentItem=(Type=Bone,Name="None")
    RotationOffset=(X=120,Y=0,Z=0)
    ExecutePin=@RigVMFunction_Sequence_4.A
   )
   n54 = &RigVMFunction_AlphaInterp(InterpSpeedDecreasing=50, InterpSpeedIncreasing=50, Scale=1, bClampResult=False, bInterpResult=False, Value=@n53.OutputParam)
   > &RigUnit_ModifyTransforms(
    ItemToModify=((Item=(Type=Bone,Name="dyn_hair_base_l"),Transform=(Rotation=(X=0,Y=0,Z=0,W=1),Translation=(X=-1,Y=1.5,Z=0),Scale3D=(X=1,Y=1,Z=1))))
    WeightMinimum=0
    ItemToModify.0.Transform.Rotation=&RigVMFunction_MathQuaternionFromEuler(Euler=(X=0,Y=0,Z=-10)).Result
    Weight=@n54.Result
   )
   > &RigUnit_ModifyTransforms(
    ItemToModify=((Item=(Type=Bone,Name="dyn_hair_base_r"),Transform=(Rotation=(X=0,Y=0,Z=0,W=1),Translation=(X=-1,Y=0,Z=0),Scale3D=(X=1,Y=1,Z=1))))
    WeightMinimum=0
    ItemToModify.0.Transform.Rotation=&RigVMFunction_MathQuaternionFromEuler(Euler=(X=0,Y=10,Z=30)).Result
    Weight=@n54.Result
   )
   n55 = &RigUnit_SphericalPoseReader(
    ActiveRegionScaleFactors=(PositiveWidth=0.5,NegativeWidth=0.5,PositiveHeight=0.5,NegativeHeight=0.5)
    ActiveRegionSize=0.1
    Debug=(bDrawDebug=true,bDraw2D=False,bDrawLocalAxes=False,DebugScale=25,DebugSegments=20,DebugThickness=0.25)
    DriverAxis=(X=0,Y=1,Z=0)
    DriverItem=(Type=Bone,Name="pelvis")
    FalloffRegionScaleFactors=(PositiveWidth=1,NegativeWidth=1,PositiveHeight=1,NegativeHeight=1)
    FalloffSize=0.25
    FlipHeightScaling=False
    FlipWidthScaling=False
    OptionalParentItem=(Type=Bone,Name="None")
    RotationOffset=(X=30,Y=0,Z=0)
    ExecutePin=@RigVMFunction_Sequence_4.B
   )
   n56 = &RigVMFunction_AlphaInterp(InterpSpeedDecreasing=50, InterpSpeedIncreasing=50, Scale=1, bClampResult=False, bInterpResult=False, Value=@n55.OutputParam)
   > &RigUnit_ModifyTransforms(
    ItemToModify=((Item=(Type=Bone,Name="dyn_hair_base_l"),Transform=(Rotation=(X=0,Y=0,Z=0,W=1),Translation=(X=0,Y=0,Z=0),Scale3D=(X=1,Y=1,Z=1))))
    WeightMinimum=0
    ItemToModify.0.Transform.Rotation=&RigVMFunction_MathQuaternionFromEuler(Euler=(X=10,Y=0,Z=20)).Result
    Weight=@n56.Result
   )
   > &RigUnit_ModifyTransforms(
    ItemToModify=((Item=(Type=Bone,Name="dyn_hair_base_r"),Transform=(Rotation=(X=0,Y=0,Z=0,W=1),Translation=(X=0,Y=0,Z=0),Scale3D=(X=1,Y=1,Z=1))))
    WeightMinimum=0
    ItemToModify.0.Transform.Rotation=&RigVMFunction_MathQuaternionFromEuler(Euler=(X=-15,Y=-15,Z=0)).Result
    Weight=@n56.Result
   )
   n57 = &RigVMFunction_MathDoubleMul(A=RigVMVariableNode() {variable=Dyn_Rotate_l}.Value, B=RigVMVariableNode() {variable=Dyn_FE}.Value)
   &RigUnit_ModifyTransforms(
    ItemToModify=((Item=(Type=Bone,Name="dyn_hair_base_b"),Transform=(Rotation=(X=0,Y=0,Z=0,W=1),Translation=(X=0,Y=-1,Z=-0.5),Scale3D=(X=1,Y=1,Z=1))))
    WeightMinimum=0
    Weight=@n57.Result
    ItemToModify.0.Transform.Rotation=&RigVMFunction_MathQuaternionFromEuler(Euler=(X=-35,Y=-25,Z=-45)).Result
    ExecutePin=@RigVMFunction_Sequence.F
   )
   > &RigUnit_ModifyTransforms(
    ItemToModify=((Item=(Type=Bone,Name="dyn_hair_base_l"),Transform=(Rotation=(X=0,Y=0,Z=0,W=1),Translation=(X=-1,Y=0,Z=0),Scale3D=(X=1,Y=1,Z=1))))
    WeightMinimum=0
    Weight=@n57.Result
    ItemToModify.0.Transform.Rotation=&RigVMFunction_MathQuaternionFromEuler(Euler=(X=-10,Y=-20,Z=0)).Result
   )
   RigVMRerouteNode(Value=@RigVMFunction_Sequence.A)
   &RigVMFunction_MathQuaternionFromEuler(Euler=(X=-35,Y=-30,Z=-30))
  ]}
]
hierarchy [
 root {loc=<0 0 0>, rot=<0 0 0 1>}
 pelvis {parent=root, loc=<0 -0.7071 93.1191>, rot=<0.0127 0.707 -0.0127 -0.707>}
 spine_01 {parent=pelvis, loc=<6.7494 -0.6388 0>, rot=<0 0 0.034 -0.9994>}
 spine_02 {parent=spine_01, loc=<7.2518 -0.035 0>, rot=<0 0 0.0238 -0.9997>}
 spine_03 {parent=spine_02, loc=<6.0104 0 0>, rot=<0 0 0.0149 -0.9999>}
 spine_04 {parent=spine_03, loc=<8.1024 0 0>, rot=<0 0 -0.0132 -0.9999>}
 spine_05 {parent=spine_04, loc=<10.0146 0 0>, rot=<0 0 -0.0015 -1>}
 neck_01 {parent=spine_05, loc=<11.3138 0.1529 0>, rot=<0 0 -0.2003 -0.9797>}
 neck_02 {parent=neck_01, loc=<4.5627 0 0>, rot=<0 0 0.0504 -0.9987>}
 head {parent=neck_02, loc=<5.3914 0 0>, rot=<0 0 0.115 -0.9934>}
 dyn_hair_base_l {parent=head, loc=<-0.2796 -1.4041 -5.6413>, rot=<0.1316 -0.9807 0.018 -0.1431>}
 dyn_hair_b_1_l {parent=dyn_hair_base_l, loc=<0 0 0>, rot=<0.018 -0.0438 0.1318 -0.9901>}
 dyn_hair_b_2_l {parent=dyn_hair_b_1_l, loc=<6.6384 0 0>, rot=<-0.0241 0.0277 0.191 -0.9809>}
 dyn_hair_1_l {parent=dyn_hair_base_l, loc=<0 0 0>, rot=<-0.0004 -0.0529 -0.2622 -0.9636>}
 dyn_hair_f_1_l {parent=dyn_hair_1_l, loc=<5.9883 0 0>, rot=<0.7664 0.1125 -0.1398 -0.6168>}
 dyn_hair_f_2_l {parent=dyn_hair_f_1_l, loc=<5.5712 0 0>, rot=<0 0 0 1>}
 dyn_hair_base_b {parent=head, loc=<-1.2277 -4.7494 0>, rot=<0.2276 -0.7126 0.6388 -0.1796>}
 dyn_hair_1_b {parent=dyn_hair_base_b, loc=<0 0 0>, rot=<-0.0598 0.001 -0.0173 -0.9981>}
 dyn_hair_2_b {parent=dyn_hair_1_b, loc=<7.0808 0 0>, rot=<0.0588 -0.0506 0.0203 -0.9968>}
 dyn_hair_base_r {parent=head, loc=<-0.4886 -1.1494 5.3976>, rot=<0.1381 -0.9599 -0.0328 0.2418>}
 dyn_hair_1_r {parent=dyn_hair_base_r, loc=<0 0 0>, rot=<-0.0087 0.118 -0.2008 -0.9725>}
 dyn_hair_f_1_r {parent=dyn_hair_1_r, loc=<5.9015 0 0>, rot=<0.5395 0.2679 -0.1862 -0.7762>}
 dyn_hair_f_2_r {parent=dyn_hair_f_1_r, loc=<6.2242 0 0>, rot=<0 0 0 1>}
 dyn_hair_b_1_r {parent=dyn_hair_base_r, loc=<0 0 0>, rot=<-0.039 0.1016 0.0887 -0.9901>}
 dyn_hair_b_2_r {parent=dyn_hair_b_1_r, loc=<6.1993 0 0>, rot=<0 0 0 1>}
 clavicle_l {parent=spine_05, loc=<4.3575 -0.1466 -2.3223>, rot=<0.0465 -0.7289 -0.009 -0.6829>}
 upperarm_l {parent=clavicle_l, loc=<15.5111 0 0>, rot=<0.0226 -0.3234 -0.0088 -0.9459>}
 clavicle_r {parent=spine_05, loc=<4.3575 -0.1466 2.3223>, rot=<0.7289 0.0465 -0.6829 0.009>}
 upperarm_r {parent=clavicle_r, loc=<-15.5111 0 0>, rot=<0.0226 -0.3234 -0.0088 -0.9459>}
]
variables [
 Dyn_FE {type=double, default=1}
 Dyn_Rotate_l {type=double, default=1}
 Dyn_crouch {type=double, default=0}
 Dyn_skydive {type=double, default=0}
]
