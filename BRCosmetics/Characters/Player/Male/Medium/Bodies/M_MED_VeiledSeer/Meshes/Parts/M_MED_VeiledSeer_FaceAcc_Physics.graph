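# Physics asset for the face accessory mesh of the medium male VeiledSeer player body, giving its dyn_hair bones secondary motion.
# Capsules on head, neck, upper spine, clavicles and upper arms form the body shapes around the dyn_hair capsules.
# Swing and twist-limited constraints build a back chain plus left and right chains that each branch front and back.
# Angularly free constraints tie both side-back tips to the back chain's tip, linking the three chains.
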PhysicsAsset /BRCosmetics/Characters/Player/Male/Medium/Bodies/M_MED_VeiledSeer/Meshes/Parts/M_MED_VeiledSeer_FaceAcc_Physics
bodies {
 clavicle_l [{
   shapeType=Capsule
   length=10.505
   radius=6.255
   shapeName=clavicle_l_capsule
   center=<7.9003 -0.4683 0.2225>
   rotation=<rot 0.8658 -94.9243 104.9629>
  }]
 clavicle_r [{
   shapeType=Capsule
   length=10.505
   radius=6.255
   shapeName=clavicle_r_capsule
   center=<-7.9017 0.8254 -0.4677>
   rotation=<rot 0.8645 -94.9242 -75.0369>
  }]
 dyn_hair_1_b [{
   shapeType=Capsule
   length=7.0848
   radius=2.9883
   shapeName=dyn_hair_1_b_capsule
   center=<4.865 -0.1827 0.5587>
   rotation=<rot 0.1533 -179.8155 -91.7365>
  }]
 dyn_hair_1_l [{
   shapeType=Capsule
   length=2.9657
   radius=2.2345
   shapeName=dyn_hair_1_l_capsule
   center=<5.2275 -0.0373 1.011>
   rotation=<rot 0.8652 -120.1009 -93.9779>
  }]
 dyn_hair_1_r [{
   shapeType=Capsule
   length=4.0773
   radius=1.8033
   shapeName=dyn_hair_1_r_capsule
   center=<6.0378 -0.6082 -1.4706>
   rotation=<rot 11.5738 87.9583 86.8656>
  }]
 dyn_hair_2_b [
  {
   shapeType=Capsule
   length=9.1186
   radius=1.3791
   shapeName=dyn_hair_2_b_capsule
   center=<6.4733 -2.29 -0.848>
   rotation=<rot 5.1276 118.1909 -87.3895>
  }
  {
   shapeType=Capsule
   length=8.25
   radius=1.75
   shapeName=dyn_hair_2_b_capsule
   center=<5.5498 2.6782 -0.3794>
   rotation=<rot 4.2574 70.4225 -90.7279>
  }
 ]
 dyn_hair_b_1_l [{
   shapeType=Capsule
   length=4.005
   radius=2.255
   shapeName=dyn_hair_b_1_l_capsule
   center=<4.0764 0 -0.0855>
   rotation=<rot -0.0001 -89.9997 -85>
  }]
 dyn_hair_b_1_r [{
   shapeType=Capsule
   length=5.75
   radius=2.25
   shapeName=dyn_hair_b_1_r_capsule
   center=<5.161 -0.0851 -0.2404>
   rotation=<rot 25.6713 -96.344 111.3175>
  }]
 dyn_hair_b_2_l [{
   shapeType=Capsule
   length=5.9621
   radius=2.3516
   shapeName=dyn_hair_b_2_l_capsule
   center=<4.3024 -0.0074 0.4373>
   rotation=<rot 0.7534 -91.8762 -69.8597>
  }]
 dyn_hair_b_2_r [{
   shapeType=Capsule
   length=5.6961
   radius=2.3787
   shapeName=dyn_hair_b_2_r_capsule
   center=<3.0686 -2.6152 -0.2551>
   rotation=<rot 11.1118 55.1551 -68.8765>
  }]
 dyn_hair_base_b [{shapeType=Capsule, length=0, radius=1.255, shapeName=dyn_hair_base_b_capsule, rotation=<rot 0 -90 0>}]
 dyn_hair_base_l [{shapeType=Capsule, length=0, radius=1.255, shapeName=dyn_hair_base_l_capsule, rotation=<rot 0 -90 0>}]
 dyn_hair_base_r [{shapeType=Capsule, length=0, radius=1.255, shapeName=dyn_hair_base_r_capsule, rotation=<rot 0 -90 0>}]
 dyn_hair_f_1_l [{
   shapeType=Capsule
   length=3.3546
   radius=2.0279
   shapeName=dyn_hair_f_1_l_capsule
   center=<2.7307 -0.1263 -0.223>
   rotation=<rot 49.6841 -82.0369 5.407>
  }]
 dyn_hair_f_1_r [
  {
   shapeType=Capsule
   length=3.5474
   radius=1.8066
   shapeName=dyn_hair_f_1_r_capsule
   center=<3.2275 1.3352 -0.8206>
   rotation=<rot -20.0308 -86.5034 -66.4284>
  }
  {
   shapeType=Capsule
   length=6.25
   radius=1.25
   shapeName=dyn_hair_f_1_r_capsule
   center=<3.1648 2.2582 -2.9659>
   rotation=<rot -74.9995 0.0003 -0.0001>
  }
 ]
 dyn_hair_f_2_l [
  {
   shapeType=Capsule
   length=6.2896
   radius=2.2901
   shapeName=dyn_hair_f_2_l_capsule
   center=<3.773 0.4409 0.7403>
   rotation=<rot 16.6041 100.9377 -91.6703>
  }
  {
   shapeType=Capsule
   length=3.5797
   radius=1
   shapeName=dyn_hair_f_2_l_capsule
   center=<0.6693 -1.6365 -1.7461>
   rotation=<rot 18.4216 80.0406 -88.0233>
  }
 ]
 dyn_hair_f_2_r [
  {
   shapeType=Capsule
   length=6.5325
   radius=2.5341
   shapeName=dyn_hair_f_2_r_capsule
   center=<3.7716 -1.6892 0.6695>
   rotation=<rot -16.6352 86.3705 -99.6749>
  }
  {
   shapeType=Capsule
   length=7.5851
   radius=1.3052
   shapeName=dyn_hair_f_2_r_capsule
   center=<0.8903 1.5605 -2.0725>
   rotation=<rot 9.8465 -100.1509 68.2458>
  }
 ]
 head [{shapeType=Capsule, length=7, radius=6.6467, shapeName=head_capsule, center=<4.4908 3.0096 -0.0185>, rotation=<rot -87.9953 7.2926 -32.2993>}]
 neck_01 [{shapeType=Capsule, length=4.1752, radius=4.4348, shapeName=neck_01_capsule, center=<2.766 1.2387 0>, rotation=<rot 0 -90.0013 -90>}]
 spine_04 [{shapeType=Capsule, length=8.25, radius=11.75, shapeName=spine_04_capsule, center=<4.146 0.2597 0>, rotation=<rot 0 -90 0>}]
 spine_05 [{
   shapeType=Capsule
   length=13.3473
   radius=7.5701
   shapeName=spine_05_capsule
   center=<2.7777 -1.161 -0.1046>
   rotation=<rot -9.4617 -90 3.3229>
  }]
 upperarm_l [{
   shapeType=Capsule
   length=14.51
   radius=4.75
   shapeName=upperarm_l_capsule
   center=<9.9066 -0.6336 0.3256>
   rotation=<rot -0.0016 -89.9986 95>
  }]
 upperarm_r [{
   shapeType=Capsule
   length=14.01
   radius=5.25
   shapeName=upperarm_r_capsule
   center=<-8.211 0.6336 -0.1703>
   rotation=<rot -0.0015 -89.9987 95>
  }]
}
constraints [
 {bone1Name=dyn_hair_1_b, bone2Name=dyn_hair_base_b, swing1LimitDegrees=60, swing2LimitDegrees=60, twistLimitDegrees=45}
 {bone1Name=dyn_hair_2_b, bone2Name=dyn_hair_1_b, swing1LimitDegrees=60, swing2LimitDegrees=60, twistLimitDegrees=75}
 {bone1Name=dyn_hair_1_l, bone2Name=dyn_hair_base_l, swing1LimitDegrees=40, swing2LimitDegrees=40, twistLimitDegrees=30}
 {bone1Name=dyn_hair_f_1_l, bone2Name=dyn_hair_1_l, swing1LimitDegrees=50, swing2LimitDegrees=45, twistLimitDegrees=15}
 {bone1Name=dyn_hair_f_2_l, bone2Name=dyn_hair_f_1_l, swing1LimitDegrees=40, swing2LimitDegrees=60, twistLimitDegrees=10}
 {bone1Name=dyn_hair_b_1_l, bone2Name=dyn_hair_base_l, swing1LimitDegrees=50, swing2LimitDegrees=50, twistLimitDegrees=20}
 {bone1Name=dyn_hair_b_2_l, bone2Name=dyn_hair_b_1_l, swing1LimitDegrees=50, swing2LimitDegrees=50, twistLimitDegrees=20}
 {bone1Name=dyn_hair_1_r, bone2Name=dyn_hair_base_r, swing1LimitDegrees=40, swing2LimitDegrees=60, twistLimitDegrees=30}
 {bone1Name=dyn_hair_f_1_r, bone2Name=dyn_hair_1_r, swing1LimitDegrees=40, swing2LimitDegrees=40, twistLimitDegrees=15}
 {bone1Name=dyn_hair_f_2_r, bone2Name=dyn_hair_f_1_r, swing1LimitDegrees=45, swing2LimitDegrees=45, twistLimitDegrees=15}
 {bone1Name=dyn_hair_b_1_r, bone2Name=dyn_hair_base_r, swing1LimitDegrees=50, swing2LimitDegrees=50, twistLimitDegrees=20}
 {bone1Name=dyn_hair_b_2_r, bone2Name=dyn_hair_b_1_r, swing1LimitDegrees=50, swing2LimitDegrees=30, twistLimitDegrees=20}
 {
  bone1Name=dyn_hair_b_2_r
  bone2Name=dyn_hair_2_b
  swing1LimitDegrees=45
  swing1Motion=Free
  swing2LimitDegrees=45
  swing2Motion=Free
  twistLimitDegrees=45
  twistMotion=Free
 }
 {
  bone1Name=dyn_hair_b_2_l
  bone2Name=dyn_hair_2_b
  swing1LimitDegrees=45
  swing1Motion=Free
  swing2LimitDegrees=45
  swing2Motion=Free
  twistLimitDegrees=45
  twistMotion=Free
 }
]
