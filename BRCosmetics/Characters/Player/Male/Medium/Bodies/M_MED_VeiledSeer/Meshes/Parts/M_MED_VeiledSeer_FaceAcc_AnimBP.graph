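# Animation Blueprint for the Veiled Seer face accessory part, driving physics-simulated dynamic hair on the player character.
# Parent and body mesh poses blend at the upper arms, then feed a Control Rig with crouch, skydive and front-end flags.
# RigidBody then simulates relative to neck_01 with state-driven gravity, and dyn_hair_2_b is constrained to spine_05 when the backbling check passes.
# Several other constraint, transform and physics nodes sit disconnected from the output.
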AnimBlueprint /BRCosmetics/Characters/Player/Male/Medium/Bodies/M_MED_VeiledSeer/Meshes/Parts/M_MED_VeiledSeer_FaceAcc_AnimBP
target_skeleton=&/M_MED_VeiledSeer_FaceAcc_Skeleton
graphs [
 AnimGraph {nodes [
   n1 = Get Dyn_Alpha_constr() -> (Dyn_Alpha_constr:float)
   "Constraint - dyn_hair_b_2_r (P)"(Alpha=@n1, ConstraintWeights_0=1) {boneToModify {boneName=dyn_hair_b_2_r}, constraintSetup [{targetBone {boneName=spine_05}}], constraintWeights [1]}
   "Constraint - dyn_hair_b_2_l (P)"(Alpha=@n1, ConstraintWeights_0=1) {boneToModify {boneName=dyn_hair_b_2_l}, constraintSetup [{targetBone {boneName=spine_05}}], constraintWeights [1]}
   Copy Pose From Mesh() {bUseAttachedParent=true}
   Layered blend per bone() {layerSetup [{branchFilters [{boneName=upperarm_l}, {boneName=upperarm_r}]}]}
   "Transform (Modify) Bone - Bone: dyn_hair_b_1_l"(Alpha=1, Translation="0,0,0", Rotation="0,0,0", Scale="1,1,1") {
    rotationMode=BMM_Additive
    rotationSpace=BCS_BoneSpace
    translationMode=BMM_Additive
    translationSpace=BCS_BoneSpace
    boneToModify {boneName=dyn_hair_b_1_l}
   }
   RigidBody!(ComponentPose, Alpha, OverrideWorldGravity:Vector="0,0,0", ExternalForce) -> (Pose) {
    bOverrideWorldGravity=true
    overridePhysicsAsset=&/M_MED_VeiledSeer_FaceAcc_Physics
    simulationSpace=BaseBoneSpace
    baseBoneRef {boneName=neck_01}
    componentAppliedLinearAccClamp=<4500 4500 4500>
    overrideWorldGravity {z=-980}
    simSpaceSettings {
     maxAngularAcceleration=75
     maxAngularVelocity=20
     maxLinearAcceleration=1750
     maxLinearVelocity=80
     velocityScaleZ=0.9
     worldAlpha=0.75
     externalLinearDragV=<1 1 1>
    }
   }
   n2 = Control Rig!(
    Source=Use cached pose 'StartCache'()
    Alpha
    Dyn_crouch:double=Get Dyn_Crouch() -> (Dyn_Crouch:double)
    Dyn_skydive:double=Get Dyn_Skydive() -> (Dyn_Skydive:double)
    Dyn_FE:double=Get Dyn_FE() -> (Dyn_FE:double)
   ) -> (Pose) {controlRigAssetReference {blueprintRigClass=&/M_MED_VeiledSeer_FaceAcc_CtrlRig.M_MED_VeiledSeer_FaceAcc_CtrlRig_C}}
   n3 = RigidBody(ComponentPose=Local To Component(LocalPose=@n2), OverrideWorldGravity=Get Override World Gravity() -> (Override World Gravity:Vector)) {
    bOverrideWorldGravity=true
    overridePhysicsAsset=&M_MED_VeiledSeer_FaceAcc_Physics
    simulationSpace=BaseBoneSpace
    baseBoneRef {boneName=neck_01}
    componentAppliedLinearAccClamp=<4500 4500 4500>
    overrideWorldGravity {z=-980}
    simSpaceSettings {
     maxAngularAcceleration=30
     maxAngularVelocity=5
     maxLinearAcceleration=1150
     maxLinearVelocity=600
     velocityScaleZ=0.9
     worldAlpha=1
     externalLinearDragV=<1 1 1>
    }
   }
   n4 = "Constraint - dyn_hair_2_b (P)"(ComponentPose=@n3, Alpha=@n1, ConstraintWeights_0=1) {boneToModify {boneName=dyn_hair_2_b}, constraintSetup [{targetBone {boneName=spine_05}}], constraintWeights [1]}
   Output Pose(Result=Component To Local(ComponentPose=@n4)) {layerGroup=None}
   n5 = Get Body() -> (Body:Skeletal Mesh Component Object Reference)
   Save cached pose 'StartCache'(Pose=Layered blend per bone(BasePose=Copy Pose From Mesh() {bCopyCurves=true, bUseAttachedParent=true}, BlendPoses_0=Copy Pose From Mesh(SourceMeshComponent=@n5)) {layerSetup [{branchFilters [{boneName=upperarm_l}, {boneName=upperarm_r}]}]}) {cacheName=StartCache}
  ]}
 EventGraph {nodes [
   Event BlueprintUpdateAnimation()
   > n1 = Sequence!(execute) -> (then_0, then_1, then_2:Exec, then_3:Exec, then_4:Exec)
   > n2 = "Gravity Override State Function (Parameter Asset) (Nativized)"(
    CharacterPartAnimInstance=Self-Reference()
    GameTimeInSeconds=GetGameTimeInSeconds()
    DebugJointName=spine_05
    Parameters=&/M_MED_VeiledSeer_Hair_Parameters
   )
   > Set Override World Gravity(Override World Gravity:Vector=@n2) -> (Output_Get:Vector)
   n3 = Branch(execute=@n1.then_1, Condition=Get bIsCrouching())
   > Set Dyn_Crouch(Dyn_Crouch:double=1) -> (Output_Get:double)
   Set Dyn_Crouch(execute=@n3.else, Dyn_Crouch=0)
   n4 = Branch(execute=@n1.then_2, Condition=CheckForBackbling())
   > Set Dyn_Alpha_constr(Dyn_Alpha_constr:float=1) -> (Output_Get:float)
   Set Dyn_Alpha_constr(execute=@n4.else, Dyn_Alpha_constr=0)
   n5 = Branch(execute=@n1.then_3, Condition=Get bIsSkyDiving())
   > Set Dyn_Skydive(Dyn_Skydive:double=1) -> (Output_Get:double)
   Set Dyn_Skydive(execute=@n5.else, Dyn_Skydive=0)
   n6 = Branch(execute=@n1.then_4, Condition=Get bIsInFrontEnd())
   > Set Dyn_FE(Dyn_FE:double=1) -> (Output_Get:double)
   Set Dyn_FE(execute=@n6.else, Dyn_FE=0)
   n7 = Event OnCharacterCustomizationLoaded()
   > Set Body(Body:Skeletal Mesh Component Object Reference=@n7.CustomBaseMesh) -> (Output_Get:Skeletal Mesh Component Object Reference)
   > Get Dynamic Material Instance(SkeletalMeshComponent=Get Array Item with EFort Custom Part Type(Array=@n7.CustomMeshParts, PartType=Face))
  ]}
]
variables [
 Body {type=SkeletalMeshComponent}
 Dyn_Alpha_constr {type=number}
 Dyn_Crouch {type=number}
 Dyn_FE {type=number, default=1}
 Dyn_Skydive {type=number}
 Override World Gravity {type=Vector, default=<0 0 -980>}
]
